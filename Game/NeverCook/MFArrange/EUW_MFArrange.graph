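# Editor utility widget that finds same-named Material Functions exposed to the material library and can unexpose them in bulk.
# One button groups all Material Functions by name in a map, uses MapIte to keep groups with two-plus exposed and one under /Game/ or /Engine/, then writes names and exposed package paths to D:\outputMF.txt.
# The other loads asset paths from the text box, clears bExposeToLibrary and calls CheckOutFile with display names, without saving.
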EditorUtilityWidgetBlueprint /Game/NeverCook/MFArrange/EUW_MFArrange
graphs [
 EventGraph {nodes [
   "On Clicked (EditorUtilityButton_14)"() {is=title K2Node_ComponentBoundEvent}
   n1 = Get Test() -> (Test:Material Function Object Reference)
   > n2 = GetAssetsByClass!(
    execute
    self:Asset Registry Interface=GetAssetRegistry()
    ClassPathName:Top Level Asset Path Structure="Get Class Path Name (Top Level Asset Path)"!(Class:Object Class Reference=Get Class(Object=@n1) -> (ReturnValue:Material Function Class Reference)) -> (ReturnValue:Top Level Asset Path Structure) {is=title K2Node_CallFunction}
    bSearchSubClasses:bool=false
   ) -> (then, OutAssetData:Array of Asset Data Structures, ReturnValue:bool)
   > n3 = For Each Loop(Array:Array of Asset Data Structures=@n2) -> (Array Element:Asset Data Structure)
   n4 = "For Each Loop (Map)"!(
    execute:Exec=@n3.Completed
    MapPin:Map of Strings to STR Asset Data Array Structures=Get Map() -> (Map:Map of Strings to STR Asset Data Array Structures)
    BreakPin:Exec
   ) -> (then:Exec, KeyPin:String, ValuePin:STR Asset Data Array Structure, CompletedPin:Exec)
   > n5 = MapIte!(execute, self, NewParam:STR Asset Data Array Structure=@n4.ValuePin) -> (then, NewParam1:bool)
   > Branch(Condition=@n5)
   > WriteStringIntoFile!(execute, WriteString:String=@n4, FilePath:String="D:\\outputMF.txt", Unique:bool=false, Clear:bool=false, __WorldContext:Object Reference) -> (then)
   > n6 = For Each Loop(Array=Break STR Asset Data Array!(STR_AssetDataArray:"STR Asset Data Array Structure (by ref)"=@n4.ValuePin) -> (AssetData_3_#1:Array of Asset Data Structures) {is=title K2Node_BreakStruct})
   > n7 = GetEditorProperty!(execute:Exec, Object:Object Reference=GetAsset(InAssetData=@n6), PropertyName:Name=bExposeToLibrary) -> (then:Exec, PropertyValue:bool, ReturnValue:bool)
   > Branch(Condition=@n7)
   > WriteStringIntoFile(WriteString="To String (Name)"(InName=Break Asset Data(AssetData=@n6).PackagePath), FilePath="D:\\outputMF.txt", Unique=false, Clear=false)
   n8 = "To String (Name)"(InName=Break Asset Data(AssetData=@n3).AssetName)
   n9 = Get Map()
   n10 = Find(TargetMap:Map of Strings to STR Asset Data Array Structures=@n9, Key=@n8) -> (Value:STR Asset Data Array Structure)
   n11 = Branch(execute=@n3.LoopBody, Condition=@n10.ReturnValue)
   > Set Asset Data(Asset Data:Array of Asset Data Structures=Break STR Asset Data Array(STR_AssetDataArray=@n10)) -> (Output_Get:Array of Asset Data Structures)
   > Remove!(execute, TargetMap:Map of Strings to STR Asset Data Array Structures=@n9, Key:"String (by ref)"=@n8) -> (then, ReturnValue:bool)
   > Add(
    TargetArray:Array of Asset Data Structures=Get Asset Data() -> (Asset Data:Array of Asset Data Structures)
    NewItem:"Asset Data Structure (by ref)"=@n3
   ) {is=title K2Node_CallArrayFunction}
   > Add!(
    execute
    TargetMap:Map of Strings to STR Asset Data Array Structures=@n9
    Key:"String (by ref)"=@n8
    Value:"STR Asset Data Array Structure (by ref)"=Make STR Asset Data Array!(AssetData_3_#1:Array of Asset Data Structures=Get Asset Data()) -> (STR_AssetDataArray:STR Asset Data Array Structure) {is=title K2Node_MakeStruct}
   ) -> (then) {is=title K2Node_CallFunction}
   > Clear(TargetArray:Array of Asset Data Structures=Get Asset Data())
   Add(
    execute=@n11.else
    TargetMap=@n9
    Key=@n8
    Value=Make STR Asset Data Array(AssetData_3_#1=Make Array("[0]":Asset Data Structure=@n3) -> (Array:Array of Asset Data Structures))
   )
   "On Clicked (EditorUtilityButton)"() {is=title K2Node_ComponentBoundEvent}
   n12 = Get Text(self:"Text Box (Multi- Line) Object Reference"=Get EditorUtilityMultiLineEditableTextBox_0() -> (EditorUtilityMultiLineEditableTextBox_0:Editor Utility Multi Line Editable Text Box Object Reference))
   > n13 = For Each Loop(Array:Array of Strings=ParseIntoArray(SourceString="To String (Text)"(InText=@n12), Delimiter="\r\n")) -> (Array Element:String)
   > n14 = LoadAsset(AssetPath=@n13)
   > PrintString(InString="To String (Boolean)"(InBool=IsValid(Object=@n14)))
   > SetEditorProperty(Object=@n14, PropertyName=bExposeToLibrary, PropertyValue=OR Boolean(A=false, B=false))
   > CheckOutFile!(execute, InFile:String=GetDisplayName(Object=@n14), bSilent:bool=true) -> (then, ReturnValue:bool)
   Event PreConstruct()
   Event Construct()
   Event Tick()
  ]}
 MapIte {nodes [
   n1 = MapIte!() -> (then, NewParam:STR Asset Data Array Structure) {is=title K2Node_FunctionEntry}
   n2 = Break STR Asset Data Array(STR_AssetDataArray=@n1)
   > n3 = Branch(Condition="integer >= integer"(A=Length(TargetArray:Array of Asset Data Structures=@n2), B=2))
   > n4 = For Each Loop(Array:Array of Asset Data Structures=@n2) -> (Array Element:Asset Data Structure)
   n5 = For Each Loop(Exec=@n4.Completed, Array:Array of Booleans=Get Should() -> (Should:Array of Booleans)) -> (Array Element:bool)
   n6 = Branch(execute=@n5.Completed, Condition="integer >= integer"(A=Get i() -> (i:int=0), B=2))
   > n7 = For Each Loop(Array:Array of Strings=Get Path() -> (Path:Array of Strings)) -> (Array Element:String)
   n8 = For Each Loop(Exec=@n7.Completed, Array:Array of Booleans=Get Should2() -> (Should2:Array of Booleans)) -> (Array Element:bool)
   n9 = Branch(execute=@n8.Completed, Condition="integer >= integer"(A=Get i_0() -> (i_0:int=0), B=1))
   > Return Node!(execute, NewParam1:bool=true)
   Return Node(execute=@n9.else, NewParam1=false)
   Branch(execute=@n8.LoopBody, Condition=@n8)
   > Increment Int(Value=Get i_0() -> (i_0=0))
   Add(
    execute=@n7.LoopBody
    TargetArray:Array of Booleans=Get Should2()
    NewItem:"Boolean (by ref)"=OR Boolean(A=StartsWith(SourceString=@n7, InPrefix=/Game/), B=StartsWith(SourceString=@n7, InPrefix=/Engine/))
   ) {is=title K2Node_CallArrayFunction}
   Return Node(execute=@n6.else, NewParam1=false)
   Branch(execute=@n5.LoopBody, Condition=@n5)
   > Increment Int(Value=Get i() -> (i=0))
   n10 = GetEditorProperty(execute=@n4.LoopBody, Object=GetAsset(InAssetData=@n4), PropertyName=bExposeToLibrary) -> (PropertyValue:"Boolean (by ref)")
   > n11 = Branch(Condition=@n10.ReturnValue)
   > Add(TargetArray=Get Should(), NewItem=@n10)
   > Add(TargetArray:Array of Strings=Get Path(), NewItem:"String (by ref)"="To String (Name)"(InName=Break Asset Data(AssetData=@n4).PackagePath))
   PrintString(execute=@n11.else, InString=GetPropertyEEEEEEEEEEEEEEEEEEEEEEEEEEEEEEEEEEEEEEEERRRRRRRRRRRRRRRRRRRRRRRRRRRRRRRRRRRRRRRRRRRRRRR)
   > Return Node(NewParam1=false)
   Return Node(execute=@n3.else, NewParam1=false)
  ]}
]
variables [
 Asset Data {type="array<AssetData>", default []}
 Map {type="map<STR_AssetDataArray>", default {}}
 Test {type=MaterialFunction, default=/Engine/Functions/Engine_MaterialFunctions01/Coordinates/1Dto2DIndex, class=~, graph=~}
]
widgets {
 info {rootWidgetClass=&CanvasPanel, widgetCount=4}
 widgets [
  CanvasPanel_21 {widgetClassPath=&CanvasPanel, widget=CanvasPanel_21}
  EditorUtilityButton_14 {
   widgetClassPath=&EditorUtilityButton
   parent=CanvasPanel_21
   widget=EditorUtilityButton_14
   slot {class=CanvasPanelSlot, properties {layoutData {alignment=<0 0>, anchors {maximum=<0 0>, minimum=<0 0>}, offsets=<ltrb 20 24 100 40>}}}
  }
  EditorUtilityMultiLineEditableTextBox_0 {
   widgetClassPath=/Script/Blutility.EditorUtilityMultiLineEditableTextBox
   bIsVariable=true
   parent=CanvasPanel_21
   widget=EditorUtilityMultiLineEditableTextBox_0
   slot {
    class=CanvasPanelSlot
    properties {layoutData {alignment=<0 0>, anchors {maximum=<0 0>, minimum=<0 0>}, offsets=<ltrb 148 28 728.2282 772.7327>}}
   }
  }
  EditorUtilityButton {
   widgetClassPath=&EditorUtilityButton
   parent=CanvasPanel_21
   widget=EditorUtilityButton
   slot {class=CanvasPanelSlot, properties {layoutData {alignment=<0 0>, anchors {maximum=<0 0>, minimum=<0 0>}, offsets=<ltrb 24 80 100 40>}}}
  }
 ]
}
